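; Dual-wield forearm pickaxe weapon Blueprint (WarmShadeWeasel) that varies its trail and impact VFX with the current swing of the combo.
; Each event reads the pawn's active montage section, so the anim trail only toggles on Swing4, the last swing.
; Impacts increment a Count that drives the main and offhand idle VFX, with Swing2 feeding the offhand.
; The impact FX SecondImpact flag clears on Swing1 and sets on Swing4 for a final burst.
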(event EventAnimTrailsNotify
  (params (bActive :type Boolean))
  ;[0.001785 0.2569 0.06994] Fist Pickaxe Last Swing Anim Trail
  (|CallParentFunction bActive)
  (bind _asfort_pawn (Utilities|Casting|CastToFortPawn (Game|GetInstigator)))
  (if (== (Animation|Montage|MontageGetCurrentSection (Components|SkeletalMesh|GetAnimInstance (Variables|/Script/Engine.Character|GetMesh _asfort_pawn)) (Animation|Montage|GetCurrentActiveMontage (Components|SkeletalMesh|GetAnimInstance (Variables|/Script/Engine.Character|GetMesh _asfort_pawn)))) "Swing4")
    (Components|Activation|Activate (CallFunction|GetAnimTrailVFXComponent) true)))

(event EventAnimTrailsDisable
  ;[0.001785 0.2569 0.06994] Fist Pickaxe Last Swing Anim Trail
  (|CallParentFunction)
  (bind _asfort_pawn (Utilities|Casting|CastToFortPawn (Game|GetInstigator)))
  (if (== (Animation|Montage|MontageGetCurrentSection (Components|SkeletalMesh|GetAnimInstance (Variables|/Script/Engine.Character|GetMesh _asfort_pawn)) (Animation|Montage|GetCurrentActiveMontage (Components|SkeletalMesh|GetAnimInstance (Variables|/Script/Engine.Character|GetMesh _asfort_pawn)))) "Swing4")
    (Components|Activation|Deactivate (CallFunction|GetAnimTrailVFXComponent))))

(event FortWeapon|EventOnPlayImpactFX
  (params (HitResult :type Hit%20Result%20Structure%20%28by%20ref%29) (ImpactPhysicalSurface :type EPhysicalSurface%20Enum :default "SurfaceType_Default") (SpawnedPSC :type FXSystem%20Component%20Object%20Reference))
  ; FX Burst Fist Rings, Final Impact Scale
  (|CallParentFunction HitResult ImpactPhysicalSurface SpawnedPSC)
  (bind _asfort_pawn (Utilities|Casting|CastToFortPawn (Game|GetInstigator)))
  (if (== (Animation|Montage|MontageGetCurrentSection (Components|SkeletalMesh|GetAnimInstance (Variables|/Script/Engine.Character|GetMesh _asfort_pawn)) (Animation|Montage|GetCurrentActiveMontage (Components|SkeletalMesh|GetAnimInstance (Variables|/Script/Engine.Character|GetMesh _asfort_pawn)))) "Swing1")
    (Effects|Components|ParticleSystem|SetBoolParameter (Variables|Default|GetImpact%20FX) "SecondImpact" false)
    (elif (== (Animation|Montage|MontageGetCurrentSection (Components|SkeletalMesh|GetAnimInstance (Variables|/Script/Engine.Character|GetMesh _asfort_pawn)) (Animation|Montage|GetCurrentActiveMontage (Components|SkeletalMesh|GetAnimInstance (Variables|/Script/Engine.Character|GetMesh _asfort_pawn)))) "Swing2")
      (bind _result (Math|Integer|IncrementInt (Variables|Default|GetCount))
        (:%20%20
          (Effects|Components|ParticleSystem|SetIntParameter (CallFunction|GetIdleOffhandVFXComponent) "Count" _result)))
      (elif (== (Animation|Montage|MontageGetCurrentSection (Components|SkeletalMesh|GetAnimInstance (Variables|/Script/Engine.Character|GetMesh _asfort_pawn)) (Animation|Montage|GetCurrentActiveMontage (Components|SkeletalMesh|GetAnimInstance (Variables|/Script/Engine.Character|GetMesh _asfort_pawn)))) "Swing3")
        _
        (elif (== (Animation|Montage|MontageGetCurrentSection (Components|SkeletalMesh|GetAnimInstance (Variables|/Script/Engine.Character|GetMesh _asfort_pawn)) (Animation|Montage|GetCurrentActiveMontage (Components|SkeletalMesh|GetAnimInstance (Variables|/Script/Engine.Character|GetMesh _asfort_pawn)))) "Swing4")
          (Utilities|FlowControl|Sequence
            (:then_0
              _)
            (:then_1
              _)
            (:then_2
              (Effects|Components|ParticleSystem|SetBoolParameter (Variables|Default|GetImpact%20FX) "SecondImpact" true)))))))
  (bind _result_2 (Math|Integer|IncrementInt (Variables|Default|GetCount))
    (:%20%20
      (Effects|Components|ParticleSystem|SetIntParameter (CallFunction|GetIdleVFXComponent) "Count" _result_2))))

(fn ConstructionScript
  (|CallParentFunction))
