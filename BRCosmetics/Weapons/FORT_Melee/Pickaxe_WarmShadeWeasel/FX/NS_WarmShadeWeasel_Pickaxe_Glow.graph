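# Niagara system for the WarmShadeWeasel pickaxe cosmetic's energy glow and burst rings.
# Glow spawns sprites and a light pulsing with loop age, adjusted for front end and time of day.
# BurstRing spawns short-lived growing meshes from a timeline curve, sliding them along a sampled skeletal mesh socket, with User.Hitted driving material parameters.
# System scripts extract and reconstruct variant colors, while Red and Black flags override the energy color.
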NiagaraSystem /BRCosmetics/Weapons/FORT_Melee/Pickaxe_WarmShadeWeasel/FX/NS_WarmShadeWeasel_Pickaxe_Glow
dependencies {
 usedDataInterfaces [&NiagaraDataInterfaceCurve, &NiagaraDataInterfaceSkeletalMesh]
 usedDynamicInputs [
  &Boolean_AndOperation
  &SetBoolByFloatComparison
  &SetBoolByIntComparison
 ]
 usedModules [
  &self:SystemSpawnScript.SystemScriptSource.SystemScriptGraph.NiagaraNodeAssignment_1.SetVariables_#1
  &self:SystemSpawnScript.SystemScriptSource.SystemScriptGraph.NiagaraNodeAssignment_3.SetVariables_#2
  &self:ExtractVariantHue_01
  &self:SystemSpawnScript.SystemScriptSource.SystemScriptGraph.NiagaraNodeAssignment_8.SetVariables_#3
  &self:ReconstructColors
  &self:SystemSpawnScript.SystemScriptSource.SystemScriptGraph.NiagaraNodeAssignment_9.SetVariables_#4
  &SystemState
  &self:SystemSpawnScript.SystemScriptSource.SystemScriptGraph.NiagaraNodeAssignment_2.SetVariables_#5
  &self:SystemSpawnScript.SystemScriptSource.SystemScriptGraph.NiagaraNodeAssignment_4.SetVariables_#6
  &Timeline
 ]
}
emitters {
 BurstRing {
  emitterUpdateScript [
   &EmitterState()
   &SpawnRate(
    SpawnRate=&FloatFromCurve(
     FloatCurve=DataInterface() {propertyValues {Curve {keys [<at 0 0 RCIM_Constant>, <at 0.0489 20 RCIM_Constant>, {arriveTangent=-157.6331, tangentMode=RCTM_User, time=0.27, value=0}]}}}
     CurveIndex=$System.Timeline.Time
    )
    Use Spawn Probability=false
   )
  ]
  particleSpawnScript [&InitializeParticle(
    Color=&MultiplyLinearColor(
     Linear Color=&MakeCustomLinearColorFromBool(
      Bool=Linked() {linkedVariable {name=System.Black?, type=NiagaraBool, defaultValue=~, description=~}}
      True Color=<rgba 0 0 0 1>
      False Color=&MakeCustomLinearColorFromBool(Bool=$System.Red, True Color=<rgba 2.5 0 0.0188 1>, False Color=$System.EnergyColorUpdated)
     )
     Scale Factor=<4 4 4 1>
    )
    Lifetime=0.35
    Mesh Scale Mode=::NewEnumerator3
    Mesh Uniform Scale=0.06
    Sprite Size Mode=::NewEnumerator3
    UsePositionOffset=true
    Write Material Random=true
   )]
  particleUpdateScript [
   &ParticleState()
   &SkeletalMeshLocation(
    Bone / Socket Index=3
    MaskSpawnGroup=false
    Mesh Sampling Type=::NewEnumerator5
    Sampled Position Offset=&MakeVector(
     X=0
     Y=0
     Z=&Lerp_Float(
      A=0
      B=-45
      Alpha=&FloatFromCurve(FloatCurve=DataInterface() {propertyValues {Curve {keys [
           {arriveTangent=0.1068, interpMode=RCIM_Cubic, leaveTangent=0.1068, tangentMode=RCTM_User, time=0.05, value=0}
           {arriveTangent=0.1671, tangentMode=RCTM_User, time=1, value=1}
          ]}}})
     )
    )
    Skeletal Mesh=$System.Skeletal Mesh
    Socket Sampling Mode=::NewEnumerator5
   )
   /SetVariables(Particles.MeshOrientation=&MultiplyQuaternion(
     Quaternion A=$Particles.SkeletalMeshLocation.SampledRotation
     Quaternion B=&MakeQuatFromAxisAngle(Axis=<0 1 0>, AngleInDegrees=180)
    )) {isSetParameters=true}
   &DynamicMaterialParameters(
    Index 0 Param 1=&MakeFloatFromVector(Vector=$Particles.Scale, Channel=::NewEnumerator0)
    Index 0 Param 2=2
    Index 0 Param 3=&FloatFromCurve(FloatCurve=DataInterface() {propertyValues {Curve {keys [<at 0 -0.3>,<at 1 -0.2>]}}}, CurveIndex=$User.Hitted)
    Index 0 Param 4=15
    Index 1 Param 1=$System.Timeline.Time
    Index 1 Param 2=&FloatFromCurve(FloatCurve=DataInterface() {propertyValues {Curve {keys [<at 0 1>,<at 1 2>]}}}, CurveIndex=$User.Hitted)
    Index 1 Param 3=0.5
    Index 1 Param 4=$Particles.Timeline.Time
    Index 2 Param 1=2
    Index 2 Param 2=2
    Index 2 Param 3=&FloatFromCurve(FloatCurve=DataInterface() {propertyValues {Curve {keys [<at 0 1>,<at 1 3>]}}}, CurveIndex=$User.Hitted)
    Module.Index0Param0Write=true
    Module.Index0Param1Write=true
    Module.Index0Param2Write=true
    Module.Index0Param3Write=true
    Module.Index1Param0Write=true
    Module.Index1Param1Write=true
    Module.Index1Param2Write=true
    Module.Index1Param3Write=true
    Module.Index2Param0Write=true
    Module.Index2Param1Write=true
    Module.Index2Param2Write=true
    Module.Index2Param3Write=true
    Write Parameter Index 1=true
    Write Parameter Index 2=true
   )
   &ScaleMeshSize(Scale Factor=&VectorFromFloat(Value=&FloatFromCurve(FloatCurve=DataInterface() {propertyValues {Curve {keys [
          {interpMode=RCIM_Cubic, leaveTangent=4.6843, tangentMode=RCTM_Break, time=0, value=0.5}
          {arriveTangent=0.297, interpMode=RCIM_Cubic, tangentMode=RCTM_Break, time=1, value=2.5}
         ]}}})))
   &ScaleColor(
    Scale Alpha=&FloatFromCurve(FloatCurve=DataInterface() {propertyValues {Curve {keys [
         {interpMode=RCIM_Cubic, tangentMode=RCTM_User, time=0, value=0}
         {arriveTangent=-1.4125, interpMode=RCIM_Cubic, leaveTangent=-1.4125, tangentMode=RCTM_User, time=0.2, value=1}
         {tangentMode=RCTM_User, time=1, value=0}
        ]}}})
    Scale RGB=&VectorFromFloat(Value=&FloatFromCurve(FloatCurve=DataInterface() {propertyValues {Curve {keys [{tangentMode=RCTM_User, time=0, value=5}, {tangentMode=RCTM_User, time=1, value=1}]}}}))
    ScaleA=true
    ScaleRGB=true
   )
   &ScaleColor(
    Scale Alpha=&MakeCustomFloatFromBool(Bool=Linked() {linkedVariable {name=System.Black?, type=NiagaraBool, defaultValue=~, description=~}}, True Float=7, False Float=0.7)
    Scale RGB=&VectorFromFloat(Value=&FNTimeOfDayScalar(Night Scale=0.6, Morning Scale=0.75, Evening Scale=0.75))
    ScaleA=true
    ScaleRGB=true
   ) {module=ScaleColor001}
  ]
  rendererClasses [&NiagaraMeshRendererProperties]
  renderers [{rendererClass=&NiagaraMeshRendererProperties, rendererIndex=0}]
 }
 Glow {
  emitterUpdateScript [
   &EmitterState(
    Inactive Response=::NewEnumerator1
    Life Cycle Mode=::NewEnumerator1
    Loop Behavior=::NewEnumerator1
    Loop Duration Mode=::NewEnumerator1
    UseLoopDelay=false
   )
   &SpawnBurst_Instantaneous(Use Spawn Probability=false, UseLoopCountLimit=false)
  ]
  particleSpawnScript [
   &InitializeParticle(
    Lifetime=5
    Position Offset=<0 0 9>
    Sprite Rotation Mode=::NewEnumerator1
    Sprite Size=&Multiply_Vector2D(
     A=&FN_FrontEndVector2Modifier(Front End:Vector2f=<58 105>, Gameplay:Vector2f=<50 97>)
     B=&Vector2DFromFloat(Value=&MakeCustomFloatFromBool(Bool=Linked() {linkedVariable {name=System.Black?, type=NiagaraBool, defaultValue=~, description=~}}, True Float=0.91, False Float=1))
    )
    Sprite Size Mode=::NewEnumerator4
    UsePositionOffset=true
    Write Material Random=true
   )
   &SpriteFacingAndAlignment(Sprite Alignment=<0 0 1>, WriteAlignment=true, WriteFacing=true)
   /SetVariables(Particles.VisibilityTag=&MakeCustomIntFromBool(Bool=Linked() {linkedVariable {name=System.Black?, type=NiagaraBool, defaultValue=~, description=~}}, True INT=1, False INT=0)) {isSetParameters=true}
  ]
  particleUpdateScript [
   &ParticleState(Kill Particles When Lifetime Has Elapsed=false, Let Infinitely Lived Particles Die When Emitter Deactivates=true)
   &ScaleColor(
    Scale Alpha=&FloatFromCurve(
     FloatCurve=DataInterface() {propertyValues {Curve {keys [
         <at 0 1 RCIM_Cubic>
         {arriveTangent=-0.1865, interpMode=RCIM_Cubic, leaveTangent=-0.1865, time=0.2, value=1.0773}
         {arriveTangent=-0.0009, interpMode=RCIM_Cubic, leaveTangent=-0.0009, time=0.5, value=0.9067}
         {arriveTangent=0.1865, interpMode=RCIM_Cubic, leaveTangent=0.1865, time=0.7, value=1.0768}
         <at 1 1 RCIM_Cubic>
        ]}}}
     CurveIndex=$System.NormalizedLoopAge
     Scale Curve=&FN_FrontEndFloatModifier(Gameplay=1, Front End=0.5)
    )
    Scale RGB=&Multiply_Vector(
     A=&MakeVectorFromLinearColorRGB(Color=&MultiplyLinearColor(
       Linear Color=&MakeCustomLinearColorFromBool(
        Bool=Linked() {linkedVariable {name=System.Black?, type=NiagaraBool, defaultValue=~, description=~}}
        True Color=<rgba 0 0 0 1>
        False Color=&MakeCustomLinearColorFromBool(
         Bool=$System.Red
         True Color=&MultiplyLinearColor(
          Linear Color=<rgba 1.4 0 0.0105 1>
          Scale Factor=&MakeFloat4FromVectorAndFloat(
           "Vector (RGB)"=&FN_FrontEndVector3Modifier(Gameplay=<1 1 1>, Front End=<0.66 0.66 0.66>)
           "Float (Alpha)"=1
          )
         )
         False Color=$System.EnergyColorUpdated
        )
       )
       Scale Factor=<0.35 0.35 0.35 1>
      ))
     B=&VectorFromFloat(Value=&FNTimeOfDayScalar(Night Scale=0.63, Morning Scale=0.8, Evening Scale=0.7))
    )
    ScaleA=true
    ScaleRGB=true
   )
   &FN_Light_Attributes(
    Color=&MultiplyLinearColor(
     Linear Color=&MakeCustomLinearColorFromBool(
      Bool=Linked() {linkedVariable {name=System.Black?, type=NiagaraBool, defaultValue=~, description=~}}
      True Color=<rgba 0 0 0 1>
      False Color=&MakeCustomLinearColorFromBool(Bool=$System.Red, True Color=<rgba 1.6 0 0.0121 1>, False Color=$System.EnergyColorUpdated)
     )
     Scale Factor=<7 7 7 1>
    )
    Create Color Attribute=true
    Create Exponent Attribute=true
    Create Radius Attribute=true
    Enable Visibility Tag=true
    Exponent=2.3
    Radius=36
    Scale Color=&Multiply_Float(
     A=&FNTimeOfDayScalar(Night Scale=0.7, Morning Scale=0.9, Evening Scale=0.85)
     B=&FloatFromCurve(
      FloatCurve=DataInterface() {propertyValues {Curve {keys [
          <at 0 1 RCIM_Cubic>
          {arriveTangent=-0.2173, interpMode=RCIM_Cubic, leaveTangent=-0.2173, time=0.2, value=1.0773}
          {arriveTangent=0.028, interpMode=RCIM_Cubic, leaveTangent=0.028, time=0.5, value=0.8914}
          {arriveTangent=0.2173, interpMode=RCIM_Cubic, leaveTangent=0.2173, time=0.7, value=1.0913}
          <at 1 1 RCIM_Cubic>
         ]}}}
      CurveIndex=$System.NormalizedLoopAge
     )
    )
    VisibilityTag:NiagaraInt32=&MakeCustomIntFromBool(Bool=Linked() {linkedVariable {name=System.Black?, type=NiagaraBool, defaultValue=~, description=~}}, True INT=1, False INT=0)
   )
  ]
  rendererClasses [&NiagaraSpriteRendererProperties, &NiagaraLightRendererProperties]
  renderers [
   {rendererClass=&NiagaraSpriteRendererProperties}
   {rendererClass=&NiagaraLightRendererProperties, rendererIndex=1}
   {rendererClass=&NiagaraSpriteRendererProperties, rendererIndex=2}
  ]
 }
}
summary {
 systemName=NS_WarmShadeWeasel_Pickaxe_Glow
 emitters [
  {emitterName=Glow, rendererClasses [&NiagaraSpriteRendererProperties, &NiagaraLightRendererProperties]}
  {emitterName=BurstRing, rendererClasses [&NiagaraMeshRendererProperties]}
 ]
 userVariables [
  User.bIsFrontEnd {type=NiagaraBool}
  User.bIsFrontEndPreview {type=NiagaraBool}
  User.Count {type=NiagaraInt32}
  User.Energy_Color {type=LinearColor, description="", defaultValue {struct=LinearColor, value=<rgba 0.7153 0.499 0.0124 1>}}
  User.Hitted {type=NiagaraFloat, defaultValue=0, description=""}
  User.RingsBurst {type=NiagaraBool}
 ]
}
