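# Translucent unlit hex-grid trail material for Niagara ribbons and sprites on the WarmShadeWeasel pickaxe.
# A tiled hex mask multiplied by boosted, panning noise forms the pattern, which particle color tints before an emissive adjustment function.
# Opacity combines the pattern, particle alpha and a procedural UV1 fade along U with a sine falloff across V.
# Depth fade softens non-mobile shading paths, and a scalar parameter blends in a pixelated trail mask.
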Material /BRCosmetics/Weapons/FORT_Melee/Pickaxe_WarmShadeWeasel/M_HexTrails
graph {
 properties {
  bAutomaticallySetUsageInEditor=false
  bUsedWithNiagaraRibbons=true
  bUsedWithNiagaraSprites=true
  blendMode=BLEND_Translucent
  shadingModel=MSM_Unlit
  usedShadingModels=Unlit
 }
 nodes [
  n1 = ParticleColor()
  n2 = TextureCoordinate()
  n3 = VectorParameter() {parameterName=Hex Tiling, defaultValue=<rgba 1.15 1.15 0 0>}
  n4 = AppendVector(A=@n3.R, B=@n3.R)
  n5 = ScalarParameter() {defaultValue=1, parameterName=MasterTile}
  n6 = TextureSampleParameter2D(UVs=Multiply(A=Multiply(A=@n2, B=@n4), B=@n5)) {parameterName=Hex Tex, samplerType=SAMPLERTYPE_Masks, texture=/Game/Accessories/FORT_Backpacks/F_MED_CyberRunner_Backpack/Textures/HexGrid_M}
  n7 = TextureCoordinate() {uTiling=0.8, vTiling=0.8}
  n8 = TextureSample(UVs=Panner(Coordinate=Multiply(A=@n7, B=@n5)) {speedX=-0.3}) {texture=&T_NoisePack_01}
  n9 = Multiply(A=@n6.R, B=Saturate(None=Multiply(A=@n8.R) {constB=4.3}))
  MaterialFunctionCall(Emissive=Multiply(A=@n1.RGB, B=@n9)) {materialFunction=&MF_FX_EmissiveAdjustments}
  n10 = TextureCoordinate() {coordinateIndex=1}
  n11 = TextureCoordinate() {coordinateIndex=1}
  n12 = ComponentMask(None=@n11) {r=true}
  n13 = ComponentMask(None=@n11) {g=true}
  n14 = Multiply(A=Multiply(A=@n9, B=@n1.A), B=Multiply(A=OneMinus(None=@n12), B=Sine(None=@n13) {period=2}) {constA=1})
  n15 = ScalarParameter() {defaultValue=50, group=Alpha, parameterName=FadeDistance, sortPriority=2}
  Multiply(
   A=LinearInterpolate(
    B=TextureSample(UVs=@n10) {texture=/BRCosmetics/Effects/Niagara/MasterMaterials/Pixelated/Textures/T_Pixelated_Trail_Mask}.R
    Alpha=ScalarParameter() {parameterName=Use Mask}
   ) {constA=1}
   B=ShadingPathSwitch(Default=DepthFade(Opacity=@n14, FadeDistance=@n15), Mobile=@n14) {desc=Migrated from FeatureLevelSwitch}
  )
 ]
}
