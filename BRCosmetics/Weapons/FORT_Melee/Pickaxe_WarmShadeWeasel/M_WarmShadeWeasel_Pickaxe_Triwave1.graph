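# Unlit translucent Niagara sprite material for a wave effect on the WarmShadeWeasel pickaxe.
# The main texture's red channel feeds a sphere mask whose radius follows the RadiusLerp dynamic parameter, with Sine and a fixed circular mask shaping the wave.
# Smoke, using switchable radial-panning or time-offset UVs, is multiplied by the blue channel and blended into that wave.
# Particle color tints emissive and scales opacity, with depth fade skipped on low quality, forward and mobile.
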Material /BRCosmetics/Weapons/FORT_Melee/Pickaxe_WarmShadeWeasel/M_WarmShadeWeasel_Pickaxe_Triwave1
graph {
 properties {bUsedWithNiagaraSprites=true, blendMode=BLEND_Translucent, pixelDepthOffsetMode=PDOM_Legacy, shadingModel=MSM_Unlit, usedShadingModels=Unlit}
 nodes [
  n1 = TextureCoordinate()
  n2 = TextureSampleParameter2D(UVs=TextureCoordinate()) {
   parameterName=Main Tex
   samplerType=SAMPLERTYPE_LinearColor
   texture=/BRCosmetics/Weapons/FORT_Melee/Pickaxe_SnowSoldierFashion/FX/Texture/T_Snow_Soldier_Wave_M
  }
  n3 = MaterialFunctionCall(X=@n2.R, Y=@n2.R) {materialFunction=&MakeFloat2}
  n4 = MaterialFunctionCall(Vector or UVs=Subtract(A=Add(A=@n1, B=@n3.Result))) {materialFunction=&VectorToRadialValue}
  n5 = Multiply(A=Time()) {constB=0.5}
  n6 = TextureSample(UVs=StaticSwitchParameter(True=Panner(Coordinate=@n4.Radial Coordinates) {speedY=-1}, False=Add(A=@n5, B=@n3.Result)) {parameterName=UseRadialUV}) {texture=&T_Inky_Smoke_Tile}
  n7 = Constant2Vector() {g=0.5, r=0.5}
  n8 = DynamicParameter() {defaultValue=<rgba 0 1 1 1>, paramNames [RadiusLerp,Param2,Param3,Param4]}
  n9 = Sine(None=SphereMask(A=@n3.Result, B=@n7, Radius=LinearInterpolate(Alpha=@n8.RadiusLerp) {constB=0.7}, Hardness=Constant()) {attenuationRadius=0.25})
  n10 = Multiply(A=@n9, B=SphereMask(A=TextureCoordinate(), B=@n7, Radius=Constant() {r=0.45}, Hardness=Constant() {r=0.5}))
  n11 = LinearInterpolate(A=Multiply(A=Multiply(A=@n6.R) {constB=5}, B=Multiply(A=@n2.B, B=@n10)), B=@n10)
  n12 = ParticleColor()
  MaterialFunctionCall(Emissive=Multiply(A=@n11, B=@n12.RGB)) {materialFunction=&MF_FX_EmissiveAdjustments}
  Multiply()
  Multiply(A=@n2.R, B=@n9)
  n13 = Multiply(A=@n11, B=@n12.A)
  n14 = DepthFade(Opacity=@n13) {fadeDistanceDefault=20}
  n15 = QualitySwitch(Default=@n14, Low=@n13, High=@n14, Medium=@n14, Epic=@n14)
  ShadingPathSwitch(Default=@n15, Deferred=@n15, Forward=@n13, Mobile=@n13)
 ]
}
