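# Unlit skeletal-mesh material for the WarmShade pickaxe that renders glowing energy armor tinted from one Energy_Color.
# Anime-style lighting takes the directional light, falling back to a sunlight parameter, for stepped Blinn/Phong highlights.
# Tech-art channels and a bleeding texture mask hue-shifted color, breathing glow, sky reflection and a time-scrolled band broken up by grid noise.
# These sum into emissive, inserted between the CharacterParent PreFX and PostFX functions.
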Material /BRCosmetics/Weapons/FORT_Melee/Pickaxe_WarmShadeWeasel/Materials/M_Pickaxe_WarmShade_Additive
graph {
 properties {
  bUseMaterialAttributes=true
  bUsedWithMeshDeformer=true
  bUsedWithSkeletalMesh=true
  ditherOpacityMask=true
  pixelDepthOffsetMode=PDOM_Legacy
  shadingModel=MSM_Unlit
  usedShadingModels=Unlit
 }
 nodes [
  AtmosphericLightVector()
  Custom() {code="return View.DirectionalLightDirection;", description=DirectionalLightVector, inputs []}
  n1 = VectorParameter() {group=Armor, parameterName=Energy_Color, defaultValue=<rgba 1 0.6332 0 1>}
  n2 = Constant() {r=0.01}
  n3 = If(A=Max(A=Max(A=@n1.R, B=@n1.R), B=@n1.B), B=@n2, "A > B"=@n1.RGB, "A == B"=@n2, "A < B"=@n2)
  Desaturation(None=@n3, Fraction=ScalarParameter() {defaultValue=-0.2, group=Armor, parameterName=ColorBleedingDarkDesaturation})
  n4 = MaterialFunctionCall() {materialFunction=&MF_CharacterParent_2023_PreFX}
  n5 = GetMaterialAttributes(None=@n4.Result) {attributeGetTypes [#1]}
  MaterialFunctionCall(
   Material=SetMaterialAttributes!(MaterialAttributes=@n5.MaterialAttributes, Emissive Color=NamedRerouteUsage(), Opacity Mask=NamedRerouteUsage()) {attributeSetTypes [#2, #3]}
   M=@n4.M
   Pre Skinned Local Position=@n4.M
   Distance Blend=@n4.M
   Fresnel=@n4.M
  ) {materialFunction=&MF_CharacterParent_2023_PostFX}
  Multiply()
  NamedRerouteDeclaration(None=@n5.Normal) {name=Normal, nodeColor=<rgba 1 0 0.6094 1>}
  n6 = NamedRerouteUsage()
  NamedRerouteDeclaration(None=SphereMask(
    A=@n6
    B=Constant() {r=1}
    Radius=ScalarParameter() {defaultValue=0.1, group=02_Anime Parameters, parameterName=SpecularHighlight_Size01, sliderMax=1, sliderMin=0.0001}
    Hardness=ScalarParameter() {defaultValue=0.2, group=02_Anime Parameters, parameterName=SpecularHighlight_Hardness01, sliderMax=1}
   ) {hardnessPercent=78}) {name=HighLight, nodeColor=<rgba 0.0078 1 0 1>}
  n7 = TextureSampleParameter2D() {group=Armor, parameterName=BleedingMask, texture=/BRCosmetics/Weapons/FORT_Melee/Pickaxe_WarmShadeWeasel/Textures/T_WarmShadeBleeding}
  NamedRerouteDeclaration(None=SmoothStep(Value=@n7.G) {constMin=0.5}) {name=DisableMask, nodeColor=<rgba 1 0 0.8672 1>}
  NamedRerouteDeclaration(None=@n4.Pre skinned Local Position) {name=PreSkinnedPos, nodeColor=<rgba 1 0 0.7734 1>}
  n8 = Custom() {code="return View.DirectionalLightDirection;", description=DirectionalLightVector, inputs []}
  n9 = DotProduct(A=@n8, B=@n8)
  n10 = CollectionParameter() {
   collection=&FortniteMaterialParameters
   parameterName=SunlightDirectionalVector
  }
  n11 = ComponentMask(None=@n10) {b=true, g=true, r=true}
  n12 = MaterialFunctionCall!(In=If(
    A=Custom() {code="return View.DirectionalLightColor;", description=DirectionalLightColor, inputs []}
    "A > B"=If(A=@n9, "A > B"=@n8, "A == B"=@n11, "A < B"=@n11) {desc=check if DirectionalLight exists}
    "A == B"=@n11
    "A < B"=@n11
   ) {desc=check if DirectionalLight color is Black}) {materialFunction=/Game/Packages/Fortress_SharedMaterials/Cinematics/Anime/MF_CINE_AnimeLightDirOverride}
  n13 = MaterialFunctionCall(Float3=@n12.Result) {materialFunction=&BreakOutFloat3Components}
  n14 = ScalarParameter() {
   desc="0.0 = the light vector is locked on the XY plane.\r\n1.0 = the actual light vector is used  \r\n(for Faces use a blend value of 0,0)"
   group=02_Anime Parameters
   parameterName=LockedZLightVector_Blend
   sliderMax=1
  }
  n15 = MaterialFunctionCall(X=@n13.R, Y=@n13.R, Z=Multiply(A=@n13.B, B=@n14)) {materialFunction=&MakeFloat3}
  n16 = StaticSwitchParameter(True=LinearInterpolate(A=@n12.Result, B=Normalize(VectorInput=@n15.Result), Alpha=VertexColor().A), False=@n12.Result) {
   desc="Lock the Z axis of the Light Vector\r\nis driven by the VertexColor A channel (0 = Regular LightVector, 1= Locked LIghtVector)"
   group=02_Anime Parameters
   parameterName=UseLockedZLightVector
   sortPriority=0
  }
  n17 = Transform(None=NamedRerouteUsage())
  NamedRerouteDeclaration(None=DotProduct(
    A=NamedRerouteDeclaration(None=@n16) {name=LightVector, nodeColor=<rgba 1 0.934 0.5103 1>}
    B=NamedRerouteDeclaration(None=@n17) {name=Normal_WS, nodeColor=<rgba 0.2083 0.2083 1 1>}
   )) {name=NdotL, nodeColor=<rgba 0 0.9297 1 1>}
  n18 = NamedRerouteUsage()
  n19 = ScalarParameter() {
   defaultValue=1
   desc="LockedZ only for the specular highlight.\r\nfor a streched highlight in on the Z axis use the Blinn highlight"
   group=02_Anime Parameters
   parameterName=SpecularHighlight_LockedZVector
   sliderMax=1
  }
  n20 = AppendVector(A=Constant2Vector() {g=1, r=1}, B=Saturate(None=OneMinus(None=@n19)))
  n21 = Multiply(A=Add(A=CameraVectorWS(), B=@n18), B=@n20)
  n22 = Multiply(A=@n18, B=@n20)
  n23 = NamedRerouteUsage()
  n24 = StaticSwitchParameter(
   True=DotProduct(A=Normalize(VectorInput=@n21), B=NamedRerouteUsage()) {desc=Blinn}
   False=DotProduct(A=Normalize(VectorInput=@n22), B=ReflectionVectorWS(CustomWorldNormal=@n23)) {desc=Phong}
  ) {
   defaultValue=true
   desc="Default is the Phong highlight.\r\nUse Blinn for a vertical streched highlight"
   group=02_Anime Parameters
   parameterName=UseBlinnHighlight
   sortPriority=0
  }
  NamedRerouteDeclaration(None=Saturate(None=@n24)) {name=BlinnGradient, nodeColor=<rgba 1 0.2109 0 1>}
  NamedRerouteDeclaration(None=ChannelMaskParameter(None=@n4.Tech Art Mask) {group=Armor, parameterName=ArmorGrid}.R) {name=GridMask, nodeColor=<rgba 0.6875 0 1 1>}
  n25 = SkyLightEnvMapSample(Direction=ReflectionVectorWS(), Roughness=Constant() {r=0.1})
  n26 = ScalarParameter() {defaultValue=1, group=Armor, parameterName=reflectIntensity}
  n27 = Multiply(A=Multiply(A=NamedRerouteUsage(), B=Multiply(A=@n25, B=@n26)), B=NamedRerouteUsage())
  n28 = NamedRerouteUsage()
  n29 = ScalarParameter() {defaultValue=0.1, group=Armor, parameterName=ColorDarkMulti}
  n30 = Sine(None=Multiply(A=@n6, B=ScalarParameter() {defaultValue=2, group=Armor, parameterName=SinMulti}) {constB=3})
  n31 = LinearInterpolate(
   A=MaterialFunctionCall(
    Hue Shift Percentage=ScalarParameter() {defaultValue=-0.01, group=Armor, parameterName=ColorDarkHueShift}
    Texture=Multiply(A=@n28, B=@n29) {constB=0.2}
   ) {materialFunction=&HueShift}.Result
   B=Multiply(A=ScalarParameter() {defaultValue=0.3, group=Armor, parameterName=BrightColorDarkMulti}, B=@n28)
   Alpha=LinearInterpolate(B=ConstantBiasScale(None=@n30), Alpha=NamedRerouteUsage()) {constA=1}
  )
  n32 = SphereMask(
   A=@n6
   B=Constant() {r=1}
   Radius=ScalarParameter() {defaultValue=0.1, group=02_Anime Parameters, parameterName=SpecularHighlight_Size, sliderMax=1, sliderMin=0.0001}
   Hardness=ScalarParameter() {defaultValue=0.2, group=02_Anime Parameters, parameterName=SpecularHighlight_Hardness, sliderMax=1}
  ) {hardnessPercent=78}
  n33 = NamedRerouteUsage()
  n34 = Multiply(
   A=NamedRerouteUsage()
   B=LinearInterpolate(
    A=ScalarParameter() {defaultValue=2, group=Armor, parameterName=MinIntensity}
    B=ScalarParameter() {defaultValue=20, group=Armor, parameterName=MaxIntensity}
    Alpha=LinearInterpolate(B=@n32, Alpha=@n33)
   ) {constA=0.5, constB=20}
  )
  n35 = Add(
   A=Multiply(A=NamedRerouteUsage(), B=ScalarParameter() {defaultValue=0.1, group=Armor, parameterName=PureColorIntensity}) {constB=10}
   B=Add(A=Multiply(A=@n31, B=@n34), B=NamedRerouteUsage())
  )
  n36 = ComponentMask(None=NamedRerouteUsage()) {b=true}
  n37 = ScalarParameter() {defaultValue=20, group=Armor, parameterName=Unit}
  n38 = Time()
  n39 = ScalarParameter() {defaultValue=-0.2, group=Armor, parameterName=SpeedFlush}
  n40 = Frac(None=Add(A=Divide(A=@n36, B=@n37), B=Multiply(A=@n38, B=@n39)))
  n41 = Constant() {r=0.5}
  n42 = Multiply(
   A=ScalarParameter() {defaultValue=0.5, group=Armor, parameterName=CenterIntensity}
   B=SphereMask(A=@n40, B=@n41, Radius=ScalarParameter() {defaultValue=0.03, group=Armor, parameterName=CenterRadius}) {attenuationRadius=0.2, hardnessPercent=0}
  )
  n43 = Power(Base=SphereMask(A=@n40, B=@n41, Radius=ScalarParameter() {defaultValue=0.15, group=Armor, parameterName=Radius}) {attenuationRadius=0.2, hardnessPercent=0})
  n44 = TextureSampleParameter2D() {
   group=Armor
   parameterName=GridRandom
   samplerType=SAMPLERTYPE_Masks
   texture=/BRCosmetics/Weapons/FORT_Melee/Pickaxe_WarmShadeWeasel/Textures/T_PunchGridRandom
  }
  n45 = Power(Base=@n44.G, Exp=ScalarParameter() {defaultValue=3, group=Armor, parameterName=PerlinPow})
  n46 = Multiply(A=@n43, B=@n45)
  n47 = Multiply(A=NamedRerouteUsage(), B=Saturate(None=Add(A=@n42, B=@n46)))
  n48 = ScalarParameter() {defaultValue=5, group=Armor, parameterName=MovingEmissiveIntensity}
  n49 = Multiply(A=@n43, B=Multiply(A=@n44.R, B=@n45))
  n50 = ScalarParameter() {defaultValue=20, group=Armor, parameterName=MovingRandomEmissiveIntensity}
  n51 = Multiply(A=NamedRerouteUsage(), B=Add(A=Multiply(A=@n47, B=@n48), B=Multiply(A=@n49, B=@n50) {constB=2}))
  n52 = NamedRerouteUsage()
  NamedRerouteDeclaration(None=Add(A=Add(A=@n27, B=@n35), B=Multiply(A=@n51, B=@n52))) {name=EM_OUT, nodeColor=<rgba 0 1 0.6016 1>}
  NamedRerouteDeclaration(None=Constant() {r=1}) {name=Opacity_OUT, nodeColor=<rgba 1 0.2812 0 1>}
  NamedRerouteDeclaration(None=ChannelMaskParameter(None=@n4.Tech Art Mask) {group=Armor, maskChannel=Blue, parameterName=ArmorLines, defaultValue=<rgba 0 0 1 0>}.R) {name=LineMask, nodeColor=<rgba 0 1 1 1>}
  n53 = Multiply(
   A=ScalarParameter() {defaultValue=2, group=Armor, parameterName=HighLightIntensity}
   B=Multiply(A=NamedRerouteUsage(), B=NamedRerouteUsage())
  )
  n54 = SmoothStep(Min=ScalarParameter() {group=Armor, parameterName=BleedMin}, Value=@n7.R)
  n55 = MaterialFunctionCall(
   Hue Shift Percentage=ScalarParameter() {defaultValue=-0.01, group=Armor, parameterName=ColorBleedingDarkHueShift}
   Texture=Multiply(A=@n3, B=ScalarParameter() {defaultValue=0.1, group=Armor, parameterName=ColorBleedingDarkMulti}) {constB=0.1}
  ) {materialFunction=&HueShift}
  n56 = NamedRerouteUsage()
  n57 = ScalarParameter() {defaultValue=1, group=Armor, parameterName=BleedingColorIntensity}
  n58 = Multiply(
   A=NamedRerouteDeclaration(None=LinearInterpolate(A=@n55.Result, B=@n3, Alpha=@n56)) {name=BleedingColorSource, nodeColor=<rgba 0.2656 1 0 1>}
   B=@n57
  ) {constB=10}
  n59 = SmoothStep(Value=Fresnel() {baseReflectFraction=0, exponent=1}) {constMax=0.8, constMin=0.5}
  n60 = Multiply(A=NamedRerouteDeclaration(None=@n3) {name=BrightColor, nodeColor=<rgba 0.0781 1 0 1>}, B=@n57)
  n61 = Sine(None=Multiply(A=ScalarParameter() {defaultValue=1, group=Armor, parameterName=BreatheSpeed}, B=Time()))
  n62 = Multiply(
   A=Add(A=Multiply(A=@n54, B=@n58), B=Multiply(A=@n59, B=@n60))
   B=LinearInterpolate(
    A=ScalarParameter() {defaultValue=1, group=Armor, parameterName=MinBreathe}
    B=ScalarParameter() {defaultValue=3, group=Armor, parameterName=MaxBreathe}
    Alpha=ConstantBiasScale(None=@n61)
   )
  )
  NamedRerouteDeclaration(None=Multiply(A=Add(A=@n53, B=@n62), B=NamedRerouteUsage())) {name=BleedColor, nodeColor=<rgba 0 1 0.5547 1>}
  ScalarParameter() {defaultValue=-0.4, group=Armor, parameterName=ColorDarkDesaturation}
  SkyAtmosphereLightDirection()
  VectorParameter() {parameterName=Color, defaultValue=<rgba 1 0.6332 0 1>}
  VectorParameter() {parameterName=ColorDark, defaultValue=<rgba 0.2882 0.1429 0 1>}
  VectorParameter() {group=Armor, parameterName=ColorPure, defaultValue=<rgba 1 0.6332 0 1>}
  VectorParameter() {group=Armor, parameterName=ColorBleedingDark, defaultValue=<rgba 0.1437 0.091 0 1>}
  VectorParameter() {group=Armor, parameterName=ReflectTint, defaultValue=<rgba 1 1 1 1>}
 ]
}
