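# Niagara swing trail for the WarmShadeWeasel pickaxe, built as one skeletal-mesh AnimTrail emitter with two ribbon renderers.
# Update sets separate hex and soft-trail colors and widths, fading over curves and gating alpha in windows of system loop age.
# System-level Black and Red flags override the processed energy color, with stronger alpha for black.
# Time of day dims RGB, and User.HideTrail zeroes trail width.
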NiagaraSystem /BRCosmetics/Weapons/FORT_Melee/Pickaxe_WarmShadeWeasel/NS_WarmShadeWeasel_Pickaxe_Trail
dependencies {
 usedDataInterfaces [&NiagaraDataInterfaceColorCurve, &NiagaraDataInterfaceCurve, &NiagaraDataInterfaceSkeletalMesh]
 usedDynamicInputs [
  &Boolean_AndOperation
  &SetBoolByFloatComparison
  &MakeFloatFromVector
  &MakeVectorFromLinearColorRGB
 ]
 usedModules [
  &self:SystemSpawnScript.SystemScriptSource.SystemScriptGraph.NiagaraNodeAssignment_1.SetVariables_#1
  &self:ExtractVariantHue_02
  &self:SystemSpawnScript.SystemScriptSource.SystemScriptGraph.NiagaraNodeAssignment_7.SetVariables_#2
  &self:ReconstructColors_01
  &self:SystemSpawnScript.SystemScriptSource.SystemScriptGraph.NiagaraNodeAssignment_8.SetVariables_#3
  &SystemState
 ]
}
emitters {Trail {
  emitterUpdateScript [
   &EmitterState()
   &FN_SpawnAnimTrail_v2()
  ]
  particleSpawnScript [
   /Game/Effects/Niagara/Modules/AnimTrail/FN_InitializeAnimTrail/FN_InitializeAnimTrailBoneIndecies(Second Bone or Socket Index=1)
   &FN_InitializeAnimTrail(
    AnimTrailWidth=&MakeCustomFloatFromBool(Bool=$User.HideTrail, True Float=0, False Float=1.5)
    Color=<rgba 2 2 2 1>
    Lifetime=0.125
    SkeletalMesh=$System.Skeletal Mesh
   ) {module=FN_InitializeAnimTrail001}
  ]
  particleUpdateScript [
   &FN_UpdateAnimTrail(SkeletalMesh=$System.Skeletal Mesh)
   /SetVariables(
    Particles.HexTrail:LinearColor=&MultiplyLinearColor(
     Linear Color=&MultiplyLinearColor(
      Linear Color=&MultiplyLinearColor(
       Linear Color=&MakeCustomLinearColorFromBool(
        Bool=Linked() {linkedVariable {name=System.Black?, type=NiagaraBool, defaultValue=~, description=~}}
        True Color=<rgba 0 0 0 1>
        False Color=&MakeCustomLinearColorFromBool(Bool=$System.Red, True Color=<rgba 3 0.0781 0.1001 1>, False Color=$System.EnergyColorUpdated)
       )
       Scale Factor=&MakeVector4_FromLinearColor(LinearColor=&ColorFromCurve(ColorCurve=DataInterface() {propertyValues {
           AlphaCurve {keys [<at 0 0.8 RCIM_Cubic>,<at 0.2544 0.2>,<at 0.838 0 RCIM_Cubic>]}
           BlueCurve {keys [<at 0 1.7 RCIM_Cubic>,<at 1 0 RCIM_Cubic>]}
           GreenCurve {keys [<at 0 1.7 RCIM_Cubic>,<at 1 0 RCIM_Cubic>]}
           RedCurve {keys [<at 0 1.7 RCIM_Cubic>,<at 1 0 RCIM_Cubic>]}
          }}))
      )
      Scale Factor=&MakeVector4(
       X=1
       Y=1
       Z=1
       W=&FloatFromCurve(
        FloatCurve=DataInterface() {propertyValues {Curve {keys [
            <at 0.25 1 RCIM_Cubic>
            {interpMode=RCIM_Cubic, tangentMode=RCTM_User, time=0.3, value=0}
            {interpMode=RCIM_Cubic, tangentMode=RCTM_User, time=0.5, value=0}
            {interpMode=RCIM_Cubic, tangentMode=RCTM_User, time=0.55, value=1}
            {interpMode=RCIM_Cubic, tangentMode=RCTM_User, time=0.67, value=1}
            <at 0.7 0 RCIM_Cubic>
           ]}}}
        CurveIndex=$System.NormalizedLoopAge
        Scale Curve=&MakeCustomFloatFromBool(Bool=Linked() {linkedVariable {name=System.Black?, type=NiagaraBool, defaultValue=~, description=~}}, True Float=2.2, False Float=1)
       )
      )
     )
     Scale Factor=&MakeFloat4FromVectorAndFloat(
      "Vector (RGB)"=&VectorFromFloat(Value=&FNTimeOfDayScalar(Night Scale=0.63, Morning Scale=0.8, Evening Scale=0.7))
      "Float (Alpha)"=1
     )
    )
    Particles.HexTrailWidth:NiagaraFloat=&Multiply_Float(A=$Particles.RibbonWidth, B=0.7)
   ) {isSetParameters=true}
   /SetVariables(
    Particles.SoftTrailColor:LinearColor=&MultiplyLinearColor(
     Linear Color=&MultiplyLinearColor(
      Linear Color=&MultiplyLinearColor(
       Linear Color=&MakeCustomLinearColorFromBool(
        Bool=Linked() {linkedVariable {name=System.Black?, type=NiagaraBool, defaultValue=~, description=~}}
        True Color=<rgba 0 0 0 1>
        False Color=&MakeCustomLinearColorFromBool(Bool=$System.Red, True Color=<rgba 1.6 0 0.0121 1>, False Color=$System.EnergyColorUpdated)
       )
       Scale Factor=&MakeVector4_FromLinearColor(LinearColor=&ColorFromCurve(ColorCurve=DataInterface() {propertyValues {
           AlphaCurve {keys [<at 0 0.3 RCIM_Cubic>,<at 0.2783 0.1>,<at 1 0 RCIM_Cubic>]}
           BlueCurve {keys [<at 0 3 RCIM_Cubic>,<at 0.9655 1>]}
           GreenCurve {keys [<at 0 3 RCIM_Cubic>,<at 0.9655 1>]}
           RedCurve {keys [<at 0 3 RCIM_Cubic>,<at 0.9655 1>]}
          }}))
      )
      Scale Factor=&MakeVector4(
       X=1
       Y=1
       Z=1
       W=&FloatFromCurve(
        FloatCurve=DataInterface() {propertyValues {Curve {keys [
            <at 0.25 1 RCIM_Cubic>
            {interpMode=RCIM_Cubic, tangentMode=RCTM_User, time=0.3, value=0}
            {interpMode=RCIM_Cubic, tangentMode=RCTM_User, time=0.5, value=0}
            {interpMode=RCIM_Cubic, tangentMode=RCTM_User, time=0.55, value=1}
            {interpMode=RCIM_Cubic, tangentMode=RCTM_User, time=0.67, value=1}
            <at 0.7 0 RCIM_Cubic>
           ]}}}
        CurveIndex=$System.NormalizedLoopAge
        Scale Curve=&MakeCustomFloatFromBool(Bool=Linked() {linkedVariable {name=System.Black?, type=NiagaraBool, defaultValue=~, description=~}}, True Float=6, False Float=1)
       )
      )
     )
     Scale Factor=&MakeFloat4FromVectorAndFloat(
      "Vector (RGB)"=&VectorFromFloat(Value=&FNTimeOfDayScalar(Night Scale=0.63, Morning Scale=0.8, Evening Scale=0.7))
      "Float (Alpha)"=1
     )
    )
    Particles.SoftTrailWidth:NiagaraFloat=&Multiply_Float(A=$Particles.RibbonWidth, B=0.8)
   ) {isSetParameters=true}
  ]
  rendererClasses [&NiagaraRibbonRendererProperties]
  renderers [
   {rendererClass=&NiagaraRibbonRendererProperties}
   {rendererClass=&NiagaraRibbonRendererProperties, rendererIndex=1}
  ]
 }}
summary {
 systemName=NS_WarmShadeWeasel_Pickaxe_Trail
 emitters [{emitterName=Trail, rendererClasses [&NiagaraRibbonRendererProperties]}]
 userVariables [
  User.Debug {type=NiagaraFloat, defaultValue=0, description=""}
  User.Energy_Color {type=LinearColor, description="", defaultValue {struct=LinearColor, value=<rgba 0.7153 0.499 0.0124 1>}}
  User.HideTrail {type=NiagaraBool}
  User.RibbonIDOffset {type=NiagaraInt32}
  User.SwingSectionIndex {type=NiagaraInt32}
 ]
}
